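# Chaos Cloth test asset in the cape v3 folder, apparently for centrifugal force behavior, built from a test plane skeletal mesh.
# A painted MaxDistance map feeds chained stretch, bending, mass, self-collision, velocity scale, damping, gravity and solver configs into LOD 0.
# Velocity scaling uses reference bone space with reduced linear, full angular and zero fictitious angular scale under the Legacy model.
# The lift/drag aerodynamics branch is disconnected from the output.
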ChaosClothAsset /BRCosmetics/Accessories/FORT_Capes/Common/Base_Cape_v3/Centrifugal_Forces/Centrifugal_Forces_Test
name=EmbeddedDataflow
graph=&self:EmbeddedDataflow
connections [
 {
  fromNode=SimulationStretchConfig
  fromPinGuid=#1
  toNode=SimulationBendingConfig
  toPinGuid=#2
 }
 {
  fromNode=SimulationBendingConfig
  fromPinGuid=#3
  toNode=SimulationMassConfig
  toPinGuid=#4
 }
 {
  fromNode=WeightMap_Lift and Drag
  fromPinGuid=#5
  toNode=SimulationAerodynamicsConfig
  toPinGuid=#6
 }
 {
  fromNode=WeightMap_Lift and Drag
  fromPin=OutputName.StringValue
  fromPinGuid=#7
  toNode=SimulationAerodynamicsConfig
  toPin=Drag.WeightMap
  toPinGuid=#8
 }
 {
  fromNode=WeightMap_Lift and Drag
  fromPin=OutputName.StringValue
  fromPinGuid=#7
  toNode=SimulationAerodynamicsConfig
  toPin=Lift.WeightMap
  toPinGuid=#9
 }
 {
  fromNode=SimulationVelocityScaleConfig
  fromPinGuid=#10
  toNode=SimulationDampingConfig
  toPinGuid=#11
 }
 {
  fromNode=SimulationMaxDistanceConfig
  fromPinGuid=#12
  toNode=SimulationStretchConfig
  toPinGuid=#13
 }
 {
  fromNode=WeightMap
  fromPin=OutputName.StringValue
  fromPinGuid=#14
  toNode=SimulationMaxDistanceConfig
  toPin=MaxDistance.WeightMap
  toPinGuid=#15
 }
 {fromNode=SkeletalMeshImport, fromPinGuid=#16, toNode=WeightMap, toPinGuid=#17}
 {
  fromNode=SimulationSelfCollisionConfig
  fromPinGuid=#18
  toNode=SimulationVelocityScaleConfig
  toPinGuid=#19
 }
 {
  fromNode=WeightMap
  fromPinGuid=#20
  toNode=SimulationMaxDistanceConfig
  toPinGuid=#21
 }
 {
  fromNode=SimulationSolverConfig
  fromPinGuid=#22
  toNode=ClothAssetTerminal
  toPin="CollectionLods[0]"
  toPinGuid=#23
 }
 {
  fromNode=SimulationMassConfig
  fromPinGuid=#24
  toNode=SimulationSelfCollisionConfig
  toPinGuid=#25
 }
 {
  fromNode=SimulationDampingConfig
  fromPinGuid=#26
  toNode=SimulationGravityConfig
  toPinGuid=#27
 }
 {
  fromNode=SimulationGravityConfig
  fromPinGuid=#28
  toNode=SimulationSolverConfig
  toPinGuid=#29
 }
]
nodes [
 ClothAssetTerminal {type=FChaosClothAssetTerminalNode_v2, guid=#30}
 SimulationStretchConfig {
  type=FChaosClothAssetSimulationStretchConfigNode
  guid=#31
  properties {
   StretchAnisoDamping=(bIsAnimatable=True,Low=1,High=1,WeightMap="StretchAnisoDamping")
   StretchStiffnessBias=(bIsAnimatable=True,Low=10000000,High=10000,WeightMap="StretchStiffnessBias")
   StretchStiffnessWarp=(bIsAnimatable=True,Low=10000000,High=10000,WeightMap="StretchStiffnessWarp")
   StretchStiffnessWeft=(bIsAnimatable=True,Low=10000000,High=10000,WeightMap="StretchStiffnessWeft")
  }
 }
 SimulationBendingConfig {
  type=FChaosClothAssetSimulationBendingConfigNode
  guid=#32
  inputPins [
   Collection {type=FManagedArrayCollection}
   FlatnessRatio.WeightMap {type=FString}
   RestAngle.WeightMap {type=FString}
   BendingStiffness.WeightMap {type=FString}
   BendingStiffnessWarp.WeightMap {type=FString}
   BendingStiffnessWeft.WeightMap {type=FString}
   BendingStiffnessBias.WeightMap {type=FString}
   BendingDamping.WeightMap {type=FString}
   BendingAnisoDamping.WeightMap {type=FString}
   AnisoBucklingRatio.WeightMap {type=FString}
   BucklingStiffness.WeightMap {type=FString}
   BucklingStiffnessWarp.WeightMap {type=FString}
   BucklingStiffnessWeft.WeightMap {type=FString}
   BucklingStiffnessBias.WeightMap {type=FString}
   BucklingRatioWeighted.WeightMap {type=FString}
  ]
  properties {
   AnisoBucklingRatio=(bIsAnimatable=True,Low=0.5,High=0.5,WeightMap="BucklingRatio")
   BendingAnisoDamping=(bIsAnimatable=True,Low=1,High=1,WeightMap="BendingAnisoDamping")
   BendingStiffnessBias=(bIsAnimatable=True,Low=1,High=100,WeightMap="BendingStiffnessBias")
   BendingStiffnessWarp=(bIsAnimatable=True,Low=1,High=100,WeightMap="BendingStiffnessWarp")
   BendingStiffnessWeft=(bIsAnimatable=True,Low=1,High=100,WeightMap="BendingStiffnessWeft")
   BucklingStiffnessBias=(bIsAnimatable=True,Low=2,High=50,WeightMap="BucklingStiffnessBias")
   BucklingStiffnessWarp=(bIsAnimatable=True,Low=2,High=50,WeightMap="BucklingStiffnessWarp")
   BucklingStiffnessWeft=(bIsAnimatable=True,Low=2,High=50,WeightMap="BucklingStiffnessWeft")
  }
 }
 SimulationMassConfig {
  type=FChaosClothAssetSimulationMassConfigNode
  guid=#33
  properties {DensityWeighted=(Low=0.35,High=0.35,WeightMap="Density")}
 }
 SimulationAerodynamicsConfig {
  type=FChaosClothAssetSimulationAerodynamicsConfigNode
  guid=#34
  properties {
   Drag=(bIsAnimatable=True,Low=0.1,High=0.5,WeightMap="Drag")
   Lift=(bIsAnimatable=True,Low=0.2,High=0.7,WeightMap="Lift")
   TurbulenceRatio=1
   WindVelocity=(X=0,Y=0,Z=0)
   WindVelocitySpace=WorldSpace
  }
 }
 SimulationDampingConfig {
  type=FChaosClothAssetSimulationDampingConfigNode
  guid=#35
  properties {
   DampingCoefficientWeighted=(bIsAnimatable=True,High=0.01,WeightMap="DampingCoefficient")
   LocalDampingAngularCoefficient=0
   LocalDampingLinearCoefficient=0
   LocalDampingSpace=CenterOfMass
  }
 }
 WeightMap_Lift and Drag {type=FChaosClothAssetWeightMapNode, guid=#36, properties {OutputName=(StringValue="LiftAndDragMap")}}
 SimulationVelocityScaleConfig {
  type=FChaosClothAssetSimulationVelocityScaleConfigNode
  guid=#37
  inputPins [Collection {type=FManagedArrayCollection}]
  outputPins [Collection {type=FManagedArrayCollection}]
  properties {
   AngularVelocityScale=1
   FictitiousAngularScale=0
   FictitiousForcesModel=Legacy
   LinearVelocityScale=(X=0.4,Y=0.4,Z=0.5)
   MaxAngularAcceleration=12000
   MaxAngularVelocity=200
   MaxLinearAcceleration=(X=60000,Y=60000,Z=60000)
   MaxLinearVelocity=(X=1000,Y=1000,Z=1000)
   MaxVelocityScale=1
   VelocityScaleSpace=ReferenceBoneSpace
   bEnableAngularAccelerationClamping=False
   bEnableAngularVelocityClamping=False
   bEnableLinearAccelerationClamping=False
   bEnableLinearVelocityClamping=False
   bWarnDuplicateProperty=True
  }
 }
 SimulationMaxDistanceConfig {
  type=FChaosClothAssetSimulationMaxDistanceConfigNode
  guid=#38
  properties {MaxDistance=(bIsAnimatable=True,High=1000,WeightMap="MaxDistance")}
 }
 WeightMap {type=FChaosClothAssetWeightMapNode, guid=#39, properties {OutputName=(StringValue="MaxDistance")}}
 SimulationSolverConfig {
  type=FChaosClothAssetSimulationSolverConfigNode
  guid=#40
  properties {MaxNumIterations=6, NumIterations=2, NumSubstepsImported=(ImportedValue=2)}
 }
 SkeletalMeshImport {
  type=FChaosClothAssetSkeletalMeshImportNode
  guid=#41
  inputPins [SkeletalMesh {type="TObjectPtr<USkeletalMesh>"}]
  outputPins [Collection {type=FManagedArrayCollection}]
  properties {
   LODIndex=0
   SectionIndex=0
   SkeletalMesh=&SkeletalMesh'&/Centrifugal_force_Test_Plane'
   UVChannel=0
   UVScale=(X=1,Y=1)
   bImportRenderMesh=True
   bImportSimMesh=True
   bImportSingleSection=False
   bSetPhysicsAsset=False
  }
 }
 SimulationSelfCollisionConfig {
  type=FChaosClothAssetSimulationSelfCollisionConfigNode
  guid=#42
  inputPins [
   Collection {type=FManagedArrayCollection}
   SelfCollisionLayers.StringValue {type=FString}
   SelfCollisionDisabledFaces.StringValue {type=FString}
   SelfCollisionEnabledKinematicFaces.StringValue {type=FString}
   SelfCollisionKinematicColliderFrictionWeighted.WeightMap {type=FString}
   SelfCollisionThicknessWeighted.WeightMap {type=FString}
  ]
  outputPins [Collection {type=FManagedArrayCollection}]
  properties {
   NumContourMinimizationPostSteps=0
   SelfCollisionDisableNeighborDistance=5
   SelfCollisionDisabledFaces=(StringValue="SelfCollisionDisabledFaces")
   SelfCollisionEnabledKinematicFaces=(StringValue="SelfCollisionEnabledKinematicFaces")
   SelfCollisionFrictionImported=()
   SelfCollisionKinematicColliderFrictionWeighted=(bIsAnimatable=True,WeightMap="SelfCollisionKinematicColliderFriction")
   SelfCollisionKinematicColliderStiffness=1
   SelfCollisionKinematicColliderThickness=0
   SelfCollisionLayers=(StringValue="SelfCollisionLayers")
   SelfCollisionProximityStiffness=1
   SelfCollisionStiffness=0.5
   SelfCollisionThicknessWeighted=(bIsAnimatable=True,Low=1,High=1,WeightMap="SelfCollisionThickness")
   bSelfCollideAgainstAllKinematicVertices=False
   bSelfCollideAgainstKinematicCollidersOnly=False
   bUseContourMinimization=True
   bUseGlobalIntersectionAnalysis=True
   bUseGlobalPostStepContours=True
   bUseSelfCollisions=True
   bUseSelfIntersections=False
   bWarnDuplicateProperty=True
  }
 }
 SimulationGravityConfig {
  type=FChaosClothAssetSimulationGravityConfigNode
  guid=#43
  properties {GravityOverrideImported=(ImportedValue=(X=0,Y=0,Z=-980.665))}
 }
]
